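# Sound cue for the looping ambient audio of a Season 21 Battle Pass REGS token collectable.
# A single looping wave player feeds the token loop sound wave straight to the output.
# It borrows the Hidden Star ambient loop attenuation from the supply drop sounds and routes through the non-critical in-game UI sound class.
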SoundCue /BattlePassS21/Audio/SFX/Collectable_S21_REGS_Token_Loop_Cue1
root=WavePlayer_1
nodes [{id=WavePlayer_1, properties {bLooping=true, soundWaveAssetPtr=&/Collectable_S21_REGS_Token_Loop}}]
properties {
 attenuationSettings=/Game/Athena/Sounds/SupplyDrops/Fort_HiddenStar_Ambient_Loop_Att
 maxDistance=2500
 soundClassObject=&Fort_UI_InGame_NonCritical
 volumeMultiplier=1
}
