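# Sound cue for the pickup of a Season 21 Battle Pass REGS token collectable.
# A single wave player feeds the root directly, playing the matching token pickup sound wave with no randomization, modulation or mixing.
# The cue uses the non-critical in-game UI sound class and has a very large max distance.
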SoundCue /BattlePassS21/Audio/SFX/Collectable_S21_REGS_Token_Pickup_Cue
root=WavePlayer_0
nodes [{id=WavePlayer_0, properties {soundWaveAssetPtr=&/Collectable_S21_REGS_Token_Pickup}}]
properties {maxDistance=2097152, soundClassObject=&Fort_UI_InGame_NonCritical}
