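# Sound cue for the party popper effect of the Season 21 Battle Pass Week 1 Quest 1.
# A Random node picks one of four party popper wave variations with equal weight.
# It plays through the non-critical in-game UI sound class with a slight volume boost and a very large max distance.
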SoundCue /BattlePassS21/Audio/SFX/Quest_S21_W1_Q1_PartyPopper_Cue
root=Random_0
nodes [
 {id=Random_0, children [WavePlayer_0,WavePlayer_1,WavePlayer_2,WavePlayer_3], properties {weights [1,1,1,1]}}
 {id=WavePlayer_0, properties {soundWaveAssetPtr=&/Quest_S21_W1_Q1_PartyPopper_03}}
 {id=WavePlayer_1, properties {soundWaveAssetPtr=&/Quest_S21_W1_Q1_PartyPopper_01}}
 {id=WavePlayer_2, properties {soundWaveAssetPtr=&/Quest_S21_W1_Q1_PartyPopper_02}}
 {id=WavePlayer_3, properties {soundWaveAssetPtr=&/Quest_S21_W1_Q1_PartyPopper_04}}
]
properties {maxDistance=2097152, soundClassObject=&Fort_UI_InGame_NonCritical, volumeMultiplier=1.2}
